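# Physics asset for the M_MED_LastHour backpack cosmetic, providing physics bodies and joints for a short chain on the pack.
# Three capsule links, dyn_chain_01 to 03, are constrained in series from a small dyn_chain_mstr body, with tight twist limits and wider swing, narrowed on one axis at the root.
# A pelvis_collider body combines two large crossed capsules, named as butt colliders, and has no constraints to the chain.
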PhysicsAsset /BRCosmetics/Accessories/FORT_Backpacks/Backpack_M_MED_LastHour/Meshes/M_MED_LastHour_Pack_Physics
bodies {
 dyn_chain_01 [{shapeType=Capsule, length=2, radius=2.7, shapeName=dyn_chain_01_capsule, center=<3 0 0>, rotation=<rot 90 0 0>}]
 dyn_chain_02 [{shapeType=Capsule, length=3.005, radius=2.005, shapeName=dyn_chain_02_capsule, center=<2.274 0 0>, rotation=<rot 90 0 0>}]
 dyn_chain_03 [{shapeType=Capsule, length=2.2463, radius=2.639, shapeName=dyn_chain_03_capsule, center=<1.5555 0 0>, rotation=<rot 90 0 0>}]
 dyn_chain_mstr [{shapeType=Capsule, length=0.505, radius=0.505, shapeName=dyn_chain_mstr_capsule, rotation=<rot 0 -90 0>}]
 pelvis_collider [
  {shapeType=Capsule, length=28, radius=11.5, shapeName=dyn_buttCollider_capsule, center=<-0.75 0 0>, rotation=<rot 0 -90 0>}
  {shapeType=Capsule, length=19, radius=12.5, shapeName=dyn_buttCollider_capsule, rotation=<rot 90 0 0>}
 ]
}
constraints [
 {bone1Name=dyn_chain_01, bone2Name=dyn_chain_mstr, swing1LimitDegrees=45, swing2LimitDegrees=15, twistLimitDegrees=5}
 {bone1Name=dyn_chain_02, bone2Name=dyn_chain_01, swing1LimitDegrees=45, swing2LimitDegrees=45, twistLimitDegrees=5}
 {bone1Name=dyn_chain_03, bone2Name=dyn_chain_02, swing1LimitDegrees=45, swing2LimitDegrees=45, twistLimitDegrees=5}
]
