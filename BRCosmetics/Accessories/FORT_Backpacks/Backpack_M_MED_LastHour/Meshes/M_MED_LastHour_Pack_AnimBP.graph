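# Animation Blueprint for the LastHour backpack, adding physics motion to its dangling chain and other simulated bones.
# Each update copies the parent mesh's spine_03 world transform onto pelvis_collider, feeding a RigidBody simulation after a small dyn_book offset.
# Skydiving events drive interpolated roll offsets on three chain bones, then AnimDynamics springs run on the lower two only when bIsInFrontEnd is true.
# BPRotate and BPTranslation are computed from spine bend but unused in the AnimGraph.
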AnimBlueprint /BRCosmetics/Accessories/FORT_Backpacks/Backpack_M_MED_LastHour/Meshes/M_MED_LastHour_Pack_AnimBP
target_skeleton=&/M_MED_LastHour_Pack_Skeleton
graphs [
 AnimGraph {nodes [
   Output Pose(Result=Use cached pose 'RbanCache'()) {layerGroup=None}
   n1 = "Transform (Modify) Bone - Bone: dyn_book"(ComponentPose=Local To Component(LocalPose=Input Pose()), Alpha=1, Translation="0,1,0", Rotation="0,0,0", Scale="1,1,1") {
    rotationMode=BMM_Additive
    translationMode=BMM_Additive
    alphaBoolBlend {blendOption=Linear}
    boneToModify {boneName=dyn_book}
    translation {y=1}
   }
   n2 = "Transform (Modify) Bone - Bone: pelvis_collider"(
    ComponentPose=@n1
    Alpha=1
    Translation=Get pelvis_tran() -> (pelvis_tran:Vector)
    Rotation=Get pelvis_rot() -> (pelvis_rot:Rotator)
    Scale="1,1,1"
   ) {
    rotationMode=BMM_Replace
    rotationSpace=BCS_WorldSpace
    translationMode=BMM_Replace
    translationSpace=BCS_WorldSpace
    alphaBoolBlend {blendOption=Linear}
    boneToModify {boneName=pelvis_collider}
   }
   Save cached pose 'Start_Cache'(Pose=Component To Local(ComponentPose=@n2)) {cacheName=Start_Cache}
   n3 = RigidBody(ComponentPose=Local To Component(LocalPose=Use cached pose 'Start_Cache'())) {
    bOverrideWorldGravity=true
    overridePhysicsAsset=&/M_MED_LastHour_Pack_Physics
    simulationSpace=BaseBoneSpace
    alphaBoolBlend {blendOption=Linear}
    baseBoneRef {boneName=root}
    componentAppliedLinearAccClamp=<5000 5000 5000>
    componentLinearAccScale=<0.45 0.45 0.45>
    componentLinearVelScale=<0.45 0.45 0.45>
    overrideWorldGravity {z=-980}
   }
   n4 = Get Skydiving() -> (Skydiving:double)
   n5 = "Transform (Modify) Bone - Bone: dyn_chain_01"(ComponentPose=@n3, Alpha=@n4, Translation="0,0,0", Rotation="0,0,20", Scale="1,1,1") {
    rotationMode=BMM_Additive
    translationMode=BMM_Additive
    alphaBoolBlend {blendOption=Linear}
    alphaScaleBiasClamp {bInterpResult=true, interpSpeedDecreasing=3, interpSpeedIncreasing=2}
    boneToModify {boneName=dyn_chain_01}
    rotation {roll=20}
   }
   n6 = "Transform (Modify) Bone - Bone: dyn_chain_02"(ComponentPose=@n5, Alpha=@n4, Translation="0,0,0", Rotation="0.000,0.000,50.000", Scale="1,1,1") {
    rotationMode=BMM_Additive
    translationMode=BMM_Additive
    alphaBoolBlend {blendOption=Linear}
    alphaScaleBiasClamp {bInterpResult=true, interpSpeedDecreasing=3, interpSpeedIncreasing=2}
    boneToModify {boneName=dyn_chain_02}
    rotation {roll=50}
   }
   n7 = "Transform (Modify) Bone - Bone: dyn_chain_03"(ComponentPose=@n6, Alpha=@n4, Translation="0,0,0", Rotation="0,0,30", Scale="1,1,1") {
    rotationMode=BMM_Additive
    translationMode=BMM_Additive
    alphaBoolBlend {blendOption=Linear}
    alphaScaleBiasClamp {bInterpResult=true, interpSpeedDecreasing=3, interpSpeedIncreasing=2}
    boneToModify {boneName=dyn_chain_03}
    rotation {roll=30}
   }
   n8 = Get bIsInFrontEnd()
   n9 = Anim Dynamics - Bone: dyn_chain_02!(ComponentPose=@n7, bAlphaBoolEnabled:bool=@n8, ExternalForce="0,0,0") -> (Pose) {
    alphaInputType=Bool
    angularDampingOverride=0.99
    angularSpringConstant=20000
    bAngularSpring=true
    bOverrideAngularDamping=true
    bUseGravityOverride=true
    simulationSpace=BoneRelative
    alphaBoolBlend {blendOption=Linear}
    boundBone {boneName=dyn_chain_02}
    componentLinearAccScale=<0.4 0.4 0.4>
    componentLinearVelScale=<0.4 0.4 0.4>
    gravityOverride {z=-980}
    physicsBodyDefinitions [{
      boundBone {boneName=dyn_chain_02}
      boxExtents=<10 10 10>
      constraintSetup {angularLimitsMax=<10 10 10>, angularLimitsMin=<-10 -10 -10>, angularTarget=<0 0 0>}
      localJointOffset=<1 0 0>
     }]
    relativeSpaceBone {boneName=root}
   }
   n10 = Anim Dynamics - Bone: dyn_chain_03!(ComponentPose=@n9, bAlphaBoolEnabled:bool=@n8, ExternalForce="0,0,0") -> (Pose) {
    alphaInputType=Bool
    angularDampingOverride=0.99
    angularSpringConstant=20000
    bAngularSpring=true
    bOverrideAngularDamping=true
    bUseGravityOverride=true
    simulationSpace=BoneRelative
    alphaBoolBlend {blendOption=Linear}
    boundBone {boneName=dyn_chain_03}
    componentLinearAccScale=<0.4 0.4 0.4>
    componentLinearVelScale=<0.4 0.4 0.4>
    gravityOverride {z=-980}
    physicsBodyDefinitions [{
      boundBone {boneName=dyn_chain_03}
      boxExtents=<10 10 10>
      constraintSetup {angularLimitsMax=<10 10 10>, angularLimitsMin=<-10 -10 -10>, angularTarget=<0 0 0>}
      localJointOffset=<1 0 0>
     }]
    relativeSpaceBone {boneName=root}
   }
   Save cached pose 'RbanCache'(Pose=Component To Local(ComponentPose=@n10)) {cacheName=RbanCache}
  ]}
 EventGraph {nodes [
   Event BlueprintUpdateAnimation()
   n1 = GetOwnerSkeletalMeshComponent()
   n2 = BreakRotator(InRot=BreakTransform(InTransform=GetDeltaTransformFromRefPose(self=@n1, BoneName=spine_05, BaseName=spine_01)).Rotation)
   > Set BPRotate(BPRotate:Rotator=MakeRotator(Roll=MapRangeClamped(Value=@n2.Yaw, InRangeA=60, InRangeB=-30, OutRangeA=10, OutRangeB=25))) -> (Output_Get:Rotator)
   > Set BPTranslation(BPTranslation:Vector=MakeVector(Y=MapRangeClamped(Value=@n2.Yaw, InRangeA=-2, InRangeB=-200, OutRangeB=25))) -> (Output_Get:Vector)
   n3 = GetPartSkeletalMeshComponent()
   > Is Valid(InputObject=@n3)
   n4 = GetAttachParent!(self:Scene Component Object Reference=@n3) -> (ReturnValue:Scene Component Object Reference)
   > Is Valid(InputObject=@n4)
   n5 = BreakTransform(InTransform=GetSocketTransform(self=@n4, InSocketName=spine_03, TransformSpace=RTS_World))
   > Set pelvis_tran(pelvis_tran:Vector=@n5) -> (Output_Get:Vector)
   > Set pelvis_rot(pelvis_rot:Rotator=@n5.Rotation) -> (Output_Get:Rotator)
   Event OnCharacterCustomizationLoaded()
   > n6 = Cast To PlayerPawn_Athena(Object=TryGetPawnOwner())
   > Set Body_Mesh(Body_Mesh=GetSkeletalMeshForPartType(self=@n6))
   Event OnBeginSkydiving()
   > Set Skydiving(Skydiving:double=1) -> (Output_Get:double)
   Event OnEndSkydiving()
   > Set Skydiving(Skydiving=0)
   Event BlueprintInitializeAnimation()
   > Set SkelMesh(SkelMesh=GetOwnerSkeletalMeshComponent())
   n7 = TryGetPawnOwner()
   > Is Valid(InputObject=@n7)
   > n8 = Cast To FortPlayerPawnAthena(Object=@n7)
   > Set PlayerPawn(PlayerPawn:Fort Player Pawn Athena Object Reference=@n8) -> (Output_Get:Fort Player Pawn Athena Object Reference)
  ]}
]
variables [
 Body_Mesh {type=SkeletalMeshComponent}
 BPRotate {type=Rotator}
 BPTranslation {type=Vector}
 pelvis_rot {type=Rotator}
 pelvis_tran {type=Vector}
 PlayerPawn {type=FortPlayerPawnAthena}
 SkelMesh {type=SkeletalMeshComponent}
 Skydiving {type=number}
]
